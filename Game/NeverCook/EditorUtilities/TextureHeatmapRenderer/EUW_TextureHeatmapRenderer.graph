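# Editor utility widget that renders a heatmap of actor descriptor locations over an embedded top-down map.
# Gather Data caches locations through GetActorDescLocations; a local World Partition variant with class filtering exists but is unwired.
# Render bins them by editable Grid Size into cell severities, passed with map bounds in a HeatmapRendererDataObject.
# GenerateHeatmapTexture draws to a render target, scaling dots by cell-to-bounds ratio, then the map shows it as overlay.
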EditorUtilityWidgetBlueprint /Game/NeverCook/EditorUtilities/TextureHeatmapRenderer/EUW_TextureHeatmapRenderer
graphs [
 DrawToRenderTarget {nodes [
   DrawToRenderTarget() {is=title K2Node_FunctionEntry}
   > Is Valid(InputObject=Get CurrentRenderTarget() -> (CurrentRenderTarget:Texture Render Target 2D Object Reference))
   > n1 = Sequence!(execute) -> (then_0, then_1, then_2:Exec)
   > n2 = SortLocationsToCells!(
    execute
    locations:Array of Vectors=Get CachedLocations() -> (CachedLocations:Array of Vectors)
    gridSize:double=Get Grid Size() -> (Grid Size:double)
    __WorldContext:Object Reference
   ) -> (then, gridCellData:"Map of Vectors to Float (double-precision)s")
   > n3 = Keys!(execute, TargetMap:"Map of Vectors to Float (double-precision)s"=@n2) -> (then, Keys:Array of Vectors)
   > Set l_locations(l_locations:Array of Vectors=@n3) -> (Output_Get:Array of Vectors)
   > n4 = Values!(execute, TargetMap:"Map of Vectors to Float (double-precision)s"=@n2) -> (then, Values:"Array of Float (double-precision)s")
   > Set l_severities(l_severities:"Array of Float (double-precision)s"=@n4) -> (Output_Get:"Array of Float (double-precision)s")
   n5 = Construct Object from Class!(
    execute=@n1.then_2
    Class=/Game/Blueprints/HeatmapRenderer/Data/HeatmapRendererDataObject.HeatmapRendererDataObject_C
    self
    Locations:Array of Vectors=Get l_locations() -> (l_locations:Array of Vectors)
    Severities:"Array of Float (double-precision)s"=Get l_severities() -> (l_severities:"Array of Float (double-precision)s")
   ) -> (then, ReturnValue:Heatmap Renderer Data Object Object Reference)
   > Set l_Data(l_Data:Heatmap Renderer Data Object Object Reference=@n5) -> (Output_Get:Heatmap Renderer Data Object Object Reference)
   > n6 = Set CurrentBounds(CurrentBounds:Box 2D Structure=Get CurrentBounds!(self:EUW Top Down Map Widget Object Reference=Get EUW_TopDownMapWidget() -> (EUW_TopDownMapWidget:EUW Top Down Map Widget Object Reference)) -> (CurrentBounds:Box 2D Structure)) -> (Output_Get:Box 2D Structure)
   > Set Bounds!(
    execute
    Bounds:Box 2D Structure=@n6
    self:Heatmap Renderer Data Object Object Reference=Get l_Data() -> (l_Data:Heatmap Renderer Data Object Object Reference)
   ) -> (then, Output_Get:Box 2D Structure)
   n7 = Get CurrentRenderTarget()
   n8 = Get SizeY!(self:Texture Render Target 2D Object Reference=@n7) -> (SizeY:int=0)
   n9 = Get SizeX!(self:Texture Render Target 2D Object Reference=@n7) -> (SizeX:int=0)
   n10 = Break Box 2D!(Box2D:"Box 2D Structure (by ref)"=Get CurrentBounds!() -> (CurrentBounds)) -> (Min:Vector 2D Structure, Max:Vector 2D Structure, bIsValid:bool=~) {is=title K2Node_BreakStruct}
   n11 = BreakVector2D(InVec=vector2d - vector2d(A:Vector 2D Structure=@n10.Max, B:Vector 2D Structure=@n10) -> (ReturnValue:Vector 2D Structure) {is=title K2Node_PromotableOperator})
   n12 = "Max (Float)"(A=@n11, B=@n11.Y)
   n13 = Get Grid Size()
   > GenerateHeatmapTexture!(
    execute
    heatmapDataObject:Heatmap Renderer Data Object Object Reference=Get l_Data()
    renderTarget:Texture Render Target 2D Object Reference=Get CurrentRenderTarget()
    dotSize:double=float * float!(A=float / float(A=@n13, B=@n12), B:int="Max (Integer)"(A=@n9, B=@n8), C:double=3) -> (ReturnValue) {is=title K2Node_PromotableOperator}
    __WorldContext:Object Reference
   ) -> (then, success:bool)
   > SetOverlayTexture!(
    execute
    self:EUW Top Down Map Widget Object Reference=Get EUW_TopDownMapWidget()
    OverlayTexture:Texture Object Reference=Get CurrentRenderTarget()
   ) -> (then)
  ]}
 EventGraph {nodes [
   "On Clicked (btnRender)"!() -> (OutputDelegate, then, widget:User Widget Object Reference, event:Pointer Event Structure) {is=title K2Node_ComponentBoundEvent}
   > DrawToRenderTarget() {is=title K2Node_CallFunction}
   "On Clicked (btnGather)"!() -> (OutputDelegate, then, widget:User Widget Object Reference, event:Pointer Event Structure) {is=title K2Node_ComponentBoundEvent}
   > PopulateActorLocationAndSeverities()
   Event Construct()
   > SetObject(self=Get SinglePropertyView_180() -> (SinglePropertyView_180:Single Property View Object Reference), NewObject=Self-Reference())
  ]}
 GetActorDescLocationsInternal {nodes [
   n1 = GetActorDescLocationsInternal!() -> (then, classSubstring:String, __WorldContext:Object Reference) {is=title K2Node_FunctionEntry}
   > Set l_class(l_class:String=@n1) -> (Output_Get:String)
   > n2 = Sequence()
   > n3 = GetActorDescs!(execute) -> (then, OutActorDescs:Array of Actor Desc Structures, ReturnValue:bool)
   > n4 = For Each Loop(Array:Array of Actor Desc Structures=@n3) -> (Array Element:Actor Desc Structure)
   n5 = Break Actor Desc!(ActorDesc:"Actor Desc Structure (by ref)"=@n4) -> (
    Guid:Guid Structure
    NativeClass:Object Class Reference
    Class:Soft Object Path Structure
    Name:Name
    Label:Name
    Bounds:Box Structure
    RuntimeGrid:Name
    bIsSpatiallyLoaded:bool=~
    bActorIsEditorOnly:bool=~
    ActorPackage:Name
    ActorPath:Name
    FolderPath:Name
    DataLayerAssets:Array of Soft Object Path Structures
   ) {is=title K2Node_BreakStruct}
   > n6 = Branch(Condition=@n5.bActorIsEditorOnly)
   Branch(execute=@n6.else, Condition=@n5.bIsSpatiallyLoaded)
   n7 = BreakSoftObjectPath(InSoftObjectPath=@n5.Class)
   > Branch(Condition=NOT Boolean(A=StartsWith(SourceString=@n7, InPrefix=/Script)))
   n8 = BreakSoftObjectPath(InSoftObjectPath=@n5.Class)
   > Add(
    TargetArray:Array of Names=Get l_packageNames() -> (l_packageNames:Array of Names)
    NewItem:"Name (by ref)"=String To Name(InString=Split(SourceString=@n8, InStr=., SearchDir=FromEnd))
   ) {is=title K2Node_CallArrayFunction}
   n9 = Break Box!(Box:"Box Structure (by ref)"=@n5.Bounds) -> (Min:Vector=~, Max:Vector=~, IsValid:bool=~) {is=title K2Node_BreakStruct}
   > Add(
    TargetArray:Array of Vectors=Get l_allLocs() -> (l_allLocs:Array of Vectors)
    NewItem:"Vector (by ref)"=vector / vector(A:Vector=vector + vector(A=@n9, B=@n9.Max), B:float=2) -> (ReturnValue:Vector)
   )
   n10 = Branch(execute=@n2.then_1, Condition="Not Equal Exactly (String)"(A=Get l_class() -> (l_class:String)))
   > n11 = For Each Loop(Array:Array of Names=Get l_packageNames()) -> (Array Element:Name)
   > n12 = GetAssetsByPackageName!(
    execute
    self:Asset Registry Interface=GetAssetRegistry()
    PackageName:Name=@n11
    bIncludeOnlyOnDiskAssets:bool=false
    bSkipARFilteredAssets:bool=true
   ) -> (then, OutAssetData:Array of Asset Data Structures, ReturnValue:bool)
   > Branch(Condition=Is Valid Index!(TargetArray:Array of Asset Data Structures=@n12, IndexToTest:int) -> (ReturnValue:bool))
   n13 = MakeLiteralName(Value=NativeParentClass)
   n14 = "Get (a ref)"(Array:Array of Asset Data Structures=@n12, Dimension 1=0) -> (Output:"Asset Data Structure (by ref)")
   > Branch(Condition=Contains(SearchIn=GetTagValue(InAssetData=@n14, InTagName=@n13), Substring=Get l_class()))
   > Add(
    TargetArray=Get l_locs() -> (l_locs:Array of Vectors)
    NewItem="Get (a ref)"(Array:Array of Vectors=Get l_allLocs(), Dimension 1=@n11.Array Index) -> (Output:"Vector (by ref)")
   )
   Set l_locs(execute=@n10.else, l_locs:Array of Vectors=Get l_allLocs()) -> (Output_Get:Array of Vectors)
   Return Node!(execute=@n2.then_2, Array:Array of Vectors=Get l_locs())
  ]}
 PopulateActorLocationAndSeverities {nodes [
   PopulateActorLocationAndSeverities() {is=title K2Node_FunctionEntry}
   > n1 = GetActorDescLocations!(execute, classSubstring:String, __WorldContext:Object Reference=Self-Reference()) -> (then, Array:Array of Vectors)
   > Set CachedLocations(CachedLocations:Array of Vectors=@n1) -> (Output_Get:Array of Vectors)
   GetActorDescLocationsInternal!(execute, self, classSubstring:String, __WorldContext:Object Reference) -> (then, Array:Array of Vectors) {is=title K2Node_CallFunction}
  ]}
]
variables [
 ActorTagForBounds {type=string, default=HeatmapBoundsActor}
 CachedLocations {type="array<Vector>"}
 CurrentBounds {type=Box2D, default {bIsValid=false, max=<0 0>, min=<0 0>}}
 CurrentRenderTarget {type=TextureRenderTarget2D, default=/Game/NeverCook/EditorUtilities/TextureHeatmapRenderer/Art/TextureHeatmapRenderer_Target2D}
 Grid Size {type=number, default=5120}
]
widgets {
 info {rootWidgetClass=&StackBox, widgetCount=6}
 widgets [
  StackBox_0 {widgetClassPath=&StackBox, widget=StackBox_0, properties {orientation=Orient_Vertical}}
  btnGather {
   widgetClassPath=&TA_Generic01_Button
   parent=StackBox_0
   widget=btnGather
   properties {label=Gather Data}
   slot {class=StackBoxSlot}
  }
  btnRender {
   widgetClassPath=&TA_Generic01_Button
   parent=StackBox_0
   widget=btnRender
   properties {label=Render Texture Heatmap}
   slot {class=StackBoxSlot}
  }
  SinglePropertyView_180 {
   widgetClassPath=&SinglePropertyView
   parent=StackBox_0
   widget=SinglePropertyView_180
   properties {propertyName=Grid Size}
   slot {class=StackBoxSlot}
  }
  HorizontalBox_0 {
   widgetClassPath=&HorizontalBox
   parent=StackBox_0
   widget=HorizontalBox_0
   slot {class=StackBoxSlot, properties {size {sizeRule=Fill, value=1}}}
  }
  EUW_TopDownMapWidget {
   widgetClassPath=/Game/NeverCook/EditorUtilities/EditorTopDownMap/EUW_TopDownMapWidget
   bIsVariable=true
   parent=HorizontalBox_0
   widget=EUW_TopDownMapWidget
   slot {class=HorizontalBoxSlot, properties {size {sizeRule=Fill, value=1}}}
  }
 ]
}
